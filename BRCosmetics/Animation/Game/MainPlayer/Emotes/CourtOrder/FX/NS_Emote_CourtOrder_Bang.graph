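# Niagara system for the bang effect of the CourtOrder emote, combining a brief swing trail with a burst of splinter debris.
# FN_AnimTrail spawns short-lived ribbons from animation bone or socket data, fading alpha and width over life.
# Splinters fires two mesh bursts, using spawn-group masks to vary cone angle and speed, then applies ramped gravity, drag and shrinking.
# SmokeLWE, Impact, Refraction and Shockwave emitters list no modules or renderers here.
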NiagaraSystem /BRCosmetics/Animation/Game/MainPlayer/Emotes/CourtOrder/FX/NS_Emote_CourtOrder_Bang
dependencies {
 usedDataInterfaces [
  &NiagaraDataInterfaceMeshRendererInfo
  &NiagaraDataInterfaceCurve
  &NiagaraDataInterfaceSocketReader
 ]
 usedDynamicInputs [
  &FNTimeOfDayColor
  &RandomRangeLinearColor
  &MultiplyLinearColor
  &MakeVector4_FromLinearColor
 ]
 usedModules [
  &self:SystemSpawnScript.SystemScriptSource.SystemScriptGraph.NiagaraNodeAssignment_0.SetVariables_#1
  &self:SystemSpawnScript.SystemScriptSource.SystemScriptGraph.NiagaraNodeAssignment_1.SetVariables_#2
  &SystemState
 ]
}
emitters {
 FN_AnimTrail {
  emitterSpawnScript [&FN_RibbonIDOffsetUtility()]
  emitterUpdateScript [
   &EmitterState(Life Cycle Mode=::NewEnumerator1, Loop Behavior=::NewEnumerator1, Loop Duration=0.7, UseLoopDelay=false)
   &FN_SpawnAnimTrail()
  ]
  particleSpawnScript [
   &FN_InitializeAnimTrailBoneIndecies()
   &FN_InitializeAnimTrailBoneIndecies(AnimTrail Selection=1, First Bone or Socket Index=2, Second Bone or Socket Index=3) {enabled=false, module=FN_InitializeAnimTrailBoneIndecies001}
   &FN_InitializeAnimTrail(Color=<rgba 0.0521 0.023 0 0.4>, Lifetime=0.24)
   &AddVelocity(Velocity=<0 400 0>) {enabled=false}
  ]
  particleUpdateScript [
   &SolveForcesAndVelocity(Clamp Velocity=false, Limit Acceleration=false)
   &FN_UpdateAnimTrail() {module=FN_UpdateAnimTrail001}
   &ScaleColor(
    Scale Alpha=&FloatFromCurve(FloatCurve=DataInterface() {propertyValues {Curve {
        defaultValue=1
        keys [{interpMode=RCIM_Cubic, tangentMode=RCTM_User, time=0, value=1}, {interpMode=RCIM_Cubic, tangentMode=RCTM_User, time=1, value=0}]
       }}})
    ScaleA=true
    ScaleRGB=true
   )
   &ScaleRibbonWidth(Ribbon Width Scale=&FloatFromCurve(FloatCurve=DataInterface() {propertyValues {Curve {keys [<at 0 1 RCIM_Cubic>, {arriveTangent=-2.6405, interpMode=RCIM_Cubic, leaveTangent=-2.6405, tangentMode=RCTM_User, time=1, value=0}]}}}))
  ]
  rendererClasses [&NiagaraRibbonRendererProperties]
  renderers [
   {rendererClass=&NiagaraRibbonRendererProperties}
   {rendererClass=&NiagaraRibbonRendererProperties, rendererIndex=1}
  ]
 }
 Impact {rendererClasses [], renderers []}
 Refraction {rendererClasses [], renderers []}
 Shockwave {rendererClasses [], renderers []}
 SmokeLWE {rendererClasses [], renderers []}
 Splinters {
  emitterUpdateScript [
   &EmitterState()
   &SpawnBurst_Instantaneous(
    Spawn Count=&RandomRangeInt(Minimum=9, Maximum=14)
    Use Spawn Probability=false
    UseLoopCountLimit=false
   )
   &SpawnBurst_Instantaneous(
    Spawn Count=&RandomRangeInt(Minimum=4, Maximum=6)
    Spawn Group=1
    Use Spawn Probability=false
    UseLoopCountLimit=false
   ) {module=SpawnBurst_Instantaneous001}
  ]
  particleSpawnScript [
   &InitializeParticle(
    Lifetime Max=0.9
    Lifetime Min=0.5
    Lifetime Mode=::NewEnumerator1
    Mesh Renderer Array Visibility Mode=::NewEnumerator3
    Mesh Renderer Info=DataInterface() {propertyValues {MeshRenderer=&self:Splinters_0.NiagaraMeshRendererProperties_0}}
    Mesh Scale Max=<0.6 0.3 0.5>
    Mesh Scale Min=<0.4 0.1 0.2>
    Mesh Scale Mode=::NewEnumerator2
    UsePositionOffset=false
    Write Material Random=true
   )
   &InitialMeshOrientation(Use Rotation Vector=false)
   &AddVelocity(
    Cone Angle=&MaskFloatBySpawnGroup(Masked Float=90, Passthrough Float=45)
    Cone Axis=<0 0 1>
    Module.UseConeFalloff=false
    Module.UseVelDistribution=false
    Override Seed=false
    Velocity Mode=::NewEnumerator2
    Velocity Speed=&MaskFloatBySpawnGroup(
     Masked Float=&RandomRangeFloat(Minimum=50, Maximum=175)
     Passthrough Float=&RandomRangeFloat(Minimum=180, Maximum=225)
    )
   )
  ]
  particleUpdateScript [
   &ParticleState()
   &ScaleMeshSize(Scale Factor=&VectorFromFloat(Value=&FloatFromCurve(FloatCurve=DataInterface() {propertyValues {Curve {
         defaultValue=1
         keys [{interpMode=RCIM_Cubic, tangentMode=RCTM_User, time=0, value=1}, {interpMode=RCIM_Cubic, tangentMode=RCTM_User, time=1, value=0}]
        }}})))
   &UpdateMeshOrientation(Rotation Rate=3)
   &GravityForce(Gravity=&MakeVector(
     X=0
     Y=0
     Z=&FloatFromCurve(
      FloatCurve=DataInterface() {propertyValues {Curve {keys [<at 0 0>,<at 0.25 1>]}}}
      Scale Curve=&RandomRangeFloat(Minimum=-325, Maximum=-600)
     )
    ))
   &Drag(
    Drag=&RandomRangeFloat(Minimum=0.8, Maximum=1.2)
    Rotational Drag=2
    Use Linear Drag=true
    Use Rotational Drag=true
   )
   &SolveForcesAndVelocity(Clamp Velocity=false, Limit Acceleration=false)
  ]
  rendererClasses [&NiagaraMeshRendererProperties]
  renderers [{rendererClass=&NiagaraMeshRendererProperties, rendererIndex=0}]
 }
}
summary {
 systemName=NS_Emote_CourtOrder_Bang
 emitters [
  {emitterName=SmokeLWE, rendererClasses []}
  {emitterName=Impact, rendererClasses []}
  {emitterName=Splinters, rendererClasses [&NiagaraMeshRendererProperties]}
  {emitterName=FN_AnimTrail, rendererClasses [&NiagaraRibbonRendererProperties]}
  {emitterName=Refraction, rendererClasses []}
  {emitterName=Shockwave, rendererClasses []}
 ]
 userVariables [User.RibbonIDOffset {type=NiagaraInt32}]
}
