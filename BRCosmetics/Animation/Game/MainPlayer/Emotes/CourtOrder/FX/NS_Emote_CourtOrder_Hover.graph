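# Niagara system for the hover effect of the CourtOrder main player emote, built around a single emitter named Hover.
# At system level it uses SystemState and a variable assignment in the spawn script, along with a FNTimeOfDayColor dynamic input that suggests time-of-day color adjustment.
# The Hover emitter lists no renderers, and the system lists no data interfaces or user variables.
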NiagaraSystem /BRCosmetics/Animation/Game/MainPlayer/Emotes/CourtOrder/FX/NS_Emote_CourtOrder_Hover
dependencies {
 usedDataInterfaces []
 usedDynamicInputs [&FNTimeOfDayColor]
 usedModules [
  &SystemState
  &self:SystemSpawnScript.SystemScriptSource.SystemScriptGraph.NiagaraNodeAssignment_0.SetVariables_#1
 ]
}
emitters {Hover {rendererClasses [], renderers []}}
summary {systemName=NS_Emote_CourtOrder_Hover, emitters [{emitterName=Hover, rendererClasses []}], userVariables []}
